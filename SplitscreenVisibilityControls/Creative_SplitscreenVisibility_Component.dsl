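; Creative splitscreen visibility component that hides or shows actors and scene components per player.
; Hide calls forward entries tagged with itself to a player component on each controller, added on demand, at minigame start or when playspace users join.
; For per-player meshes, it recursively duplicates a component and children, maps each copy by player and hides it from other local players.
; Minigame end clears its references, and playspace binding retries after a delay if none is found.
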(event Minigame|EventHandleMinigameStarted
  (Networking|CanExecuteCosmeticEvents
    (:True
      (bind _array_element (Default|ForEachFortLocalPlayerController)
        (:Loop%20Body
          (CallFunction|AddPlayerComponenttoPlayer _array_element)
          (CallFunction|BindMinigame))))))

(event Minigame|EventHandleMinigameEnded
  (Networking|CanExecuteCosmeticEvents
    (:True
      (bind _array_element (Default|ForEachFortLocalPlayerController)
        (:Loop%20Body
          (CallFunction|RemoveComponentReferencesfromPlayer _array_element))))))

(event Custom|Attempt%20Rebind%20Minigame
  (Utilities|FlowControl|Delay 2.0)
  (CallFunction|BindMinigame))

(event Custom|PlayspaceUserAdded
  (params (AddedUser :type Playspace%20User%20Structure%20%28by%20ref%29))
  (bind (_userid _leafplayspace _playerstatecached _controllercached) (Utilities|Struct|BreakPlayspaceUser AddedUser))
  (if (Utilities|IsValid _controllercached :!class /Script/Engine.KismetSystemLibrary)
    (bind _asfort_player_controller (Utilities|Casting|CastToFortPlayerController _controllercached))
    (CallFunction|AddPlayerComponenttoPlayer _asfort_player_controller)))

(fn AddPlayerComponentToPlayer
  (params (Player%20Controller :type Fort%20Player%20Controller%20Object%20Reference))
  (Utilities|IsValid Player%20Controller
    (:Is%20Valid
      (Components|AddControllerComponent Player%20Controller "/SplitscreenVisibilityControls/Creative_SplitscreenVisibility_Player_Component.Creative_SplitscreenVisibility_Player_Component_C"))))

(fn BindMinigame
  (bind _returnvalue (Class|PlayspaceLibrary|GetPlayspaceforActor self (Components|GetOwner)))
  (Utilities|IsValid _returnvalue
    (:Is%20Valid
      (Playspace|BindEventtoNotifyPlayspaceUserAdded _returnvalue (EventDispatchers|CreateEvent :!event PlayspaceUserAdded)))
    (:Is%20Not%20Valid
      (CallFunction|AttemptRebindMinigame))))

(fn CreateAllIndividualComponents
  (params (Component%20To%20Clone :type Mesh%20Component%20Object%20Reference))
  (bind _array_element (Default|ForEachFortLocalPlayerController)
    (:Loop%20Body
      (CallFunction|CreateIndividuallyVisibleComponent _array_element Component%20To%20Clone)
      (CallFunction|HideSceneComponentfromPlayer _array_element Component%20To%20Clone true))))

(fn CreateIndividuallyVisibleComponent
  (params (Player :type Controller%20Object%20Reference) (Component%20To%20Clone :type Mesh%20Component%20Object%20Reference))
  (bind _returned_component (CallFunction|DuplicateComponentwithChildren Component%20To%20Clone (Components|GetOwner Component%20To%20Clone) (Transformation|GetAttachParent Component%20To%20Clone)))
  (Utilities|Map|Add (Variables|Default|GetUniqueComponentMap) (Variables|GetPlayer) _returned_component)
  (bind _array_element (Default|ForEachFortLocalPlayerController)
    (:Loop%20Body
      (if (!= _array_element (Variables|GetPlayer))
        (CallFunction|HideSceneComponentfromPlayer _array_element _returned_component true)))))

(fn DuplicateComponentWithChildren
  (params (Component%20to%20Duplicate :type Scene%20Component%20Object%20Reference) (Actor%20Owner :type Actor%20Object%20Reference) (In%20Attach%20Parent :type Scene%20Component%20Object%20Reference))
  ;[0.04091 0.15 0.04635 0.5] The Duplicate funciton does not grab children, so this function will recursively duplicate a component and all of it's children, grandchildren, etc.
  (bind _spawnedcomponent (Fortnite|Utilities|TransientDuplicateSceneComponent Component%20to%20Duplicate Actor%20Owner In%20Attach%20Parent))
  (for _array_element (Components|GetChildrenComponents Component%20to%20Duplicate false)
    (CallFunction|DuplicateComponentwithChildren _array_element (Variables|GetActor%20Owner) _spawnedcomponent))
  (return _spawnedcomponent))

(fn GetUniqueComponentForPlayer
  (params (Player :type Controller%20Object%20Reference))
  (bind (_value _returnvalue) (Utilities|Map|Find (Variables|Default|GetUniqueComponentMap) Player))
  (return _value _returnvalue))

(fn HideActorFromPlayer
  (params (Player :type Controller%20Object%20Reference) (Actor :type Actor%20Object%20Reference) (ShouldBeHidden :type Boolean))
  (bind _actor (Variables|GetActor))
  (bind _self self)
  (bind _splitscreenvisibilityactorentry (Utilities|Struct|MakeSplitscreenVisibilityActorEntry _self _actor))
  (bind _output_get (LocalVariables|SetPlayerComponent (Actor|GetComponentbyClass Player "/SplitscreenVisibilityControls/Creative_SplitscreenVisibility_Player_Component.Creative_SplitscreenVisibility_Player_Component_C")))
  (Utilities|IsValid _output_get
    (:Is%20Valid
      _)
    (:Is%20Not%20Valid
      (bind _asfort_player_controller (Utilities|Casting|CastToFortPlayerController Player))
      (CallFunction|AddPlayerComponenttoPlayer _asfort_player_controller)
      (bind _output_get_2 (LocalVariables|SetPlayerComponent (Actor|GetComponentbyClass _asfort_player_controller "/SplitscreenVisibilityControls/Creative_SplitscreenVisibility_Player_Component.Creative_SplitscreenVisibility_Player_Component_C")))
      (Utilities|IsValid _output_get_2
        (:Is%20Valid
          _))))
  (if ShouldBeHidden
    (Class|CreativeSplitscreenVisibilityPlayerComponent|HideActorfromPlayer (LocalVariables|GetPlayerComponent) _splitscreenvisibilityactorentry)
    (else
      (Class|CreativeSplitscreenVisibilityPlayerComponent|ShowActortoPlayer (LocalVariables|GetPlayerComponent) _splitscreenvisibilityactorentry))))

(fn HideOtherIndividualComponentsFromPlayer
  (params (Player%20Controller :type Controller%20Object%20Reference) (Original%20Component :type Mesh%20Component%20Object%20Reference))
  (bind _player_controller (Variables|GetPlayer%20Controller))
  ; Used when Joining in progress to hide existing comps from new player
  (CallFunction|HideSceneComponentfromPlayer Player%20Controller Original%20Component true)
  (bind (_keypin _valuepin) (Utilities|Map|ForEachLoop(Map) (Variables|Default|GetUniqueComponentMap)))
  (if (!= _keypin _player_controller)
    (CallFunction|HideSceneComponentfromPlayer _player_controller _valuepin true)))

(fn HideSceneComponentFromPlayer
  (params (Player :type Controller%20Object%20Reference) (Component :type Scene%20Component%20Object%20Reference) (ShouldBeHidden :type Boolean))
  (bind _component (Variables|GetComponent))
  (bind _self self)
  (bind _splitscreenvisibilitycomponententry (Utilities|Struct|MakeSplitscreenVisibilityComponentEntry _self _component))
  (bind _output_get (LocalVariables|SetPlayerComponent (Actor|GetComponentbyClass Player "/SplitscreenVisibilityControls/Creative_SplitscreenVisibility_Player_Component.Creative_SplitscreenVisibility_Player_Component_C")))
  (Utilities|IsValid _output_get
    (:Is%20Valid
      _)
    (:Is%20Not%20Valid
      (bind _asfort_player_controller (Utilities|Casting|CastToFortPlayerController Player))
      (CallFunction|AddPlayerComponenttoPlayer _asfort_player_controller)
      (bind _output_get_2 (LocalVariables|SetPlayerComponent (Actor|GetComponentbyClass _asfort_player_controller "/SplitscreenVisibilityControls/Creative_SplitscreenVisibility_Player_Component.Creative_SplitscreenVisibility_Player_Component_C")))
      (Utilities|IsValid _output_get_2
        (:Is%20Valid
          _))))
  (if ShouldBeHidden
    (Class|CreativeSplitscreenVisibilityPlayerComponent|HideSceneComponentfromPlayer (LocalVariables|GetPlayerComponent) _splitscreenvisibilitycomponententry)
    (else
      (Class|CreativeSplitscreenVisibilityPlayerComponent|ShowSceneComponenttoPlayer (LocalVariables|GetPlayerComponent) _splitscreenvisibilitycomponententry))))

(fn RemoveAllIndividualComponents
  (params (Cloned%20Component :type Mesh%20Component%20Object%20Reference))
  (bind _array_element (Default|ForEachFortLocalPlayerController)
    (:Loop%20Body
      (CallFunction|RemoveIndividuallyVisibleComponent _array_element)
      (CallFunction|HideSceneComponentfromPlayer _array_element Cloned%20Component))))

(fn RemoveComponentReferencesFromPlayer
  (params (Player%20Controller :type Fort%20Player%20Controller%20Object%20Reference))
  (Utilities|IsValid Player%20Controller
    (:Is%20Valid
      (Class|CreativeSplitscreenVisibilityPlayerComponent|RemoveAllDeviceComponentReferences (Actor|GetComponentbyClass Player%20Controller "/SplitscreenVisibilityControls/Creative_SplitscreenVisibility_Player_Component.Creative_SplitscreenVisibility_Player_Component_C") self))))

(fn RemoveComponentWithChildren
  (params (Component%20To%20Delete :type Scene%20Component%20Object%20Reference))
  (for _array_element (Components|GetChildrenComponents Component%20To%20Delete false)
    (CallFunction|RemoveComponentwithChildren _array_element))
  (Components|DestroyComponent (Variables|GetComponent%20To%20Delete)))

(fn RemoveIndividuallyVisibleComponent
  (params (Player :type Controller%20Object%20Reference))
  (bind _output_get (LocalVariables|SetStoredComponent (Utilities|Map|Find (Variables|Default|GetUniqueComponentMap) Player)))
  (CallFunction|HideSceneComponentfromPlayer (Variables|GetPlayer) _output_get)
  (Utilities|Map|Remove (Variables|Default|GetUniqueComponentMap) (Variables|GetPlayer))
  (CallFunction|RemoveComponentwithChildren (LocalVariables|GetStoredComponent)))

(fn RemovePlayerFromAllComponents
  (params (Player%20to%20Remove :type Controller%20Object%20Reference))
  (bind (_keypin _valuepin) (Utilities|Map|ForEachLoop(Map) (Variables|Default|GetUniqueComponentMap)))
  (CallFunction|HideSceneComponentfromPlayer Player%20to%20Remove _valuepin))
