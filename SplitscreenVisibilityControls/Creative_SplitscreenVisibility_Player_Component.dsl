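; Player component that tracks and applies per-player hiding of actors and scene components for splitscreen visibility devices.
; Hide functions store unique device/target pairs, then hide the target only for the owner's FortPlayerController.
; Show functions remove the calling device's entry and unhide only when no other entry lists that target, allowing overlapping devices.
; RemoveAllDeviceComponentReferences purges one device's actor entries without unhiding them or touching the component list.
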(fn HideActorFromPlayer
  (params (ActorToHide :type Splitscreen%20Visibility%20Actor%20Entry%20Structure))
  (bind (_splitscreendevice_2_e569fd2e402d4a451049128cb3372982 _hiddenactor_5_6ce473bb4b47db53227b22a27780e6bd) (Utilities|Struct|BreakSplitscreenVisibilityActorEntry ActorToHide))
  (LocalVariables|SetOrigin%20Device _splitscreendevice_2_e569fd2e402d4a451049128cb3372982)
  (LocalVariables|SetActor%20Being%20Hidden _hiddenactor_5_6ce473bb4b47db53227b22a27780e6bd)
  (for _array_element (Variables|Default|GetActorsToHide)
    (if (and (== (LocalVariables|GetOrigin%20Device) (Utilities|Struct|BreakSplitscreenVisibilityActorEntry _array_element)) (== (.HIddenActor_5_6CE473BB4B47DB53227B22A27780E6BD (Utilities|Struct|BreakSplitscreenVisibilityActorEntry _array_element)) (LocalVariables|GetActor%20Being%20Hidden)))
      (return)))
  (Utilities|Array|Add (Variables|Default|GetActorsToHide) (Variables|GetActorToHide))
  (bind _asfort_player_controller (Utilities|Casting|CastToFortPlayerController (Fortnite|GetControllerfromActor (Components|GetOwner))))
  (PlayerController|SetActorHiddenControllerOnly _asfort_player_controller (LocalVariables|GetActor%20Being%20Hidden) true))

(fn HideSceneComponentFromPlayer
  (params (ComponentToHide :type Splitscreen%20Visibility%20Component%20Entry%20Structure))
  (bind (_splitscreendevice_5_6240a7f4419121e0f23a77b92ffa9534 _hiddencomponent_7_700be65f445d87134640bc9044b9d62a) (Utilities|Struct|BreakSplitscreenVisibilityComponentEntry ComponentToHide))
  (LocalVariables|SetOrigin%20Device _splitscreendevice_5_6240a7f4419121e0f23a77b92ffa9534)
  (LocalVariables|SetComponent%20Being%20Hidden _hiddencomponent_7_700be65f445d87134640bc9044b9d62a)
  (for _array_element (Variables|Default|GetComponentsToHide)
    (if (and (== (LocalVariables|GetOrigin%20Device) (Utilities|Struct|BreakSplitscreenVisibilityComponentEntry _array_element)) (== (.HiddenComponent_7_700BE65F445D87134640BC9044B9D62A (Utilities|Struct|BreakSplitscreenVisibilityComponentEntry _array_element)) (LocalVariables|GetComponent%20Being%20Hidden)))
      (return)))
  (Utilities|Array|Add (Variables|Default|GetComponentsToHide) (Variables|GetComponentToHide))
  (bind _asfort_player_controller (Utilities|Casting|CastToFortPlayerController (Fortnite|GetControllerfromActor (Components|GetOwner))))
  (PlayerController|SetCompHiddenControllerOnly _asfort_player_controller (LocalVariables|GetComponent%20Being%20Hidden) true true))

(fn RemoveAllDeviceComponentReferences
  (params (Component%20To%20Remove :type Creative%20Splitscreen%20Visibility%20Component%20Object%20Reference))
  (for _array_element (Variables|Default|GetActorsToHide)
    (if (not (== (Utilities|Struct|BreakSplitscreenVisibilityActorEntry _array_element) (Variables|GetComponent%20To%20Remove)))
      (Utilities|Array|Add (LocalVariables|GetTempActorList) _array_element)))
  (Variables|Default|SetActorsToHide (LocalVariables|GetTempActorList)))

(fn ShowActorToPlayer
  (params (ActorToShow :type Splitscreen%20Visibility%20Actor%20Entry%20Structure))
  (bind (_splitscreendevice_2_e569fd2e402d4a451049128cb3372982 _hiddenactor_5_6ce473bb4b47db53227b22a27780e6bd) (Utilities|Struct|BreakSplitscreenVisibilityActorEntry ActorToShow))
  (LocalVariables|SetOrigin%20Device _splitscreendevice_2_e569fd2e402d4a451049128cb3372982)
  (LocalVariables|SetActor%20Being%20Removed _hiddenactor_5_6ce473bb4b47db53227b22a27780e6bd)
  (for _array_element (Variables|Default|GetActorsToHide)
    (if (not (and (== (LocalVariables|GetOrigin%20Device) (Utilities|Struct|BreakSplitscreenVisibilityActorEntry _array_element)) (== (.HIddenActor_5_6CE473BB4B47DB53227B22A27780E6BD (Utilities|Struct|BreakSplitscreenVisibilityActorEntry _array_element)) (LocalVariables|GetActor%20Being%20Removed))))
      (Utilities|Array|Add (LocalVariables|GetTempListOfActors) _array_element)))
  (Variables|Default|SetActorsToHide (LocalVariables|GetTempListOfActors))
  ; Since 2 devices could hide the same actor, only show if we verify that the actor is unlisted after removal
  (for _array_element_2 (Variables|Default|GetActorsToHide)
    (if (== (LocalVariables|GetActor%20Being%20Removed) (.HIddenActor_5_6CE473BB4B47DB53227B22A27780E6BD (Utilities|Struct|BreakSplitscreenVisibilityActorEntry _array_element_2)))
      (return)))
  (bind _asfort_player_controller (Utilities|Casting|CastToFortPlayerController (Fortnite|GetControllerfromActor (Components|GetOwner))))
  (PlayerController|SetActorHiddenControllerOnly _asfort_player_controller (LocalVariables|GetActor%20Being%20Removed) false))

(fn ShowSceneComponentToPlayer
  (params (ComponentToShow :type Splitscreen%20Visibility%20Component%20Entry%20Structure))
  (bind (_splitscreendevice_5_6240a7f4419121e0f23a77b92ffa9534 _hiddencomponent_7_700be65f445d87134640bc9044b9d62a) (Utilities|Struct|BreakSplitscreenVisibilityComponentEntry ComponentToShow))
  (LocalVariables|SetOrigin%20Device _splitscreendevice_5_6240a7f4419121e0f23a77b92ffa9534)
  (LocalVariables|SetComponent%20Being%20Removed _hiddencomponent_7_700be65f445d87134640bc9044b9d62a)
  (for _array_element (Variables|Default|GetComponentsToHide)
    (if (not (and (== (LocalVariables|GetOrigin%20Device) (Utilities|Struct|BreakSplitscreenVisibilityComponentEntry _array_element)) (== (.HiddenComponent_7_700BE65F445D87134640BC9044B9D62A (Utilities|Struct|BreakSplitscreenVisibilityComponentEntry _array_element)) (LocalVariables|GetComponent%20Being%20Removed))))
      (Utilities|Array|Add (LocalVariables|GetTempListOfComponents) _array_element)))
  (Variables|Default|SetComponentsToHide (LocalVariables|GetTempListOfComponents))
  ; Since 2 devices could hide the same actor, only show if we verify that the actor is unlisted after removal
  (for _array_element_2 (Variables|Default|GetComponentsToHide)
    (if (== (LocalVariables|GetComponent%20Being%20Removed) (.HiddenComponent_7_700BE65F445D87134640BC9044B9D62A (Utilities|Struct|BreakSplitscreenVisibilityComponentEntry _array_element_2)))
      (return)))
  (bind _asfort_player_controller (Utilities|Casting|CastToFortPlayerController (Fortnite|GetControllerfromActor (Components|GetOwner))))
  (PlayerController|SetCompHiddenControllerOnly _asfort_player_controller (LocalVariables|GetComponent%20Being%20Removed) false true))
